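# Sound cue for the player entering a portapotty hideable.
# A single wave player feeds the enter sound through a modulator that randomizes volume slightly, with no pitch variation set.
# The cue plays in the gadgets and items sound class at a reduced volume, using the mono traps attenuation preset for spatial falloff.
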SoundCue /Game/Sounds/Fort_Hideables/Portapotty/Hideable_Portapotty_Enter_Cue
root=Modulator_0
nodes [
 {id=Modulator_0, children [WavePlayer_4], properties {volumeMax=1.7, volumeMin=1.6}}
 {id=WavePlayer_4, properties {soundWaveAssetPtr=&/Hideable_Portapotty_Enter_01}}
]
properties {
 attenuationSettings=/Game/Sounds/Attenuation/Fort_Traps_Mono_Att
 maxDistance=6000
 soundClassObject=&Fort_Gadgets_Items
 volumeMultiplier=0.5714
}
